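# Sound cue for the IO Brute NPC's voice line when its charge punch lands.
# A Random node picks one of five equally weighted wave variations and is excluded from branch culling.
# The cue uses the NPC voice sound class, with attenuation and chatter concurrency assets from the IO Brute ability set.
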SoundCue /NPCLibrary/Audio/IO_Brute/VO/Cues/IOBrute_2021_ChargePunchLand_Cue
root=Random_0
nodes [
 {
  id=Random_0
  children [WavePlayer_0,WavePlayer_1,WavePlayer_2,WavePlayer_3,WavePlayer_4]
  properties {bShouldExcludeFromBranchCulling=true, weights [1,1,1,1,1]}
 }
 {id=WavePlayer_0, properties {soundWaveAssetPtr=/NPCLibrary/Audio/IO_Brute/VO/Waves/IOBrute_2021_ChargePunchLand_010}}
 {id=WavePlayer_1, properties {soundWaveAssetPtr=/NPCLibrary/Audio/IO_Brute/VO/Waves/IOBrute_2021_ChargePunchLand_011}}
 {id=WavePlayer_2, properties {soundWaveAssetPtr=/NPCLibrary/Audio/IO_Brute/VO/Waves/IOBrute_2021_ChargePunchLand_012}}
 {id=WavePlayer_3, properties {soundWaveAssetPtr=/NPCLibrary/Audio/IO_Brute/VO/Waves/IOBrute_2021_ChargePunchLand_013}}
 {id=WavePlayer_4, properties {soundWaveAssetPtr=/NPCLibrary/Audio/IO_Brute/VO/Waves/IOBrute_2021_ChargePunchLand_014}}
]
properties {
 attenuationSettings=/NPCLibrary/Audio/IO_Brute/VO/IO_Brute_Ability_Att
 maxDistance=4000
 soundClassObject=&Fort_VO_NPC
 volumeMultiplier=1
 concurrencySet [/NPCLibrary/Audio/IO_Brute/VO/IO_Brute_Ability_Chatter_Conc]
}
