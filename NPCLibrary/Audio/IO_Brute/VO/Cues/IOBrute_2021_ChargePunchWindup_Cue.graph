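# Sound cue for the IO Brute NPC's charge punch windup voice line.
# A Random node picks one of three equally weighted wave variants and is excluded from branch culling.
# Playback uses the NPC VO sound class with a boosted volume, plus the Brute's ability attenuation and ability chatter concurrency settings.
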SoundCue /NPCLibrary/Audio/IO_Brute/VO/Cues/IOBrute_2021_ChargePunchWindup_Cue
root=Random_0
nodes [
 {id=Random_0, children [WavePlayer_0,WavePlayer_1,WavePlayer_2], properties {bShouldExcludeFromBranchCulling=true, weights [1,1,1]}}
 {id=WavePlayer_0, properties {soundWaveAssetPtr=/NPCLibrary/Audio/IO_Brute/VO/Waves/IOBrute_2021_ChargePunchWindup_010}}
 {id=WavePlayer_1, properties {soundWaveAssetPtr=/NPCLibrary/Audio/IO_Brute/VO/Waves/IOBrute_2021_ChargePunchWindup_011}}
 {id=WavePlayer_2, properties {soundWaveAssetPtr=/NPCLibrary/Audio/IO_Brute/VO/Waves/IOBrute_2021_ChargePunchWindup_012}}
]
properties {
 attenuationSettings=/NPCLibrary/Audio/IO_Brute/VO/IO_Brute_Ability_Att
 maxDistance=4000
 soundClassObject=&Fort_VO_NPC
 volumeMultiplier=2
 concurrencySet [/NPCLibrary/Audio/IO_Brute/VO/IO_Brute_Ability_Chatter_Conc]
}
